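; Music encounter Blueprint for the Gifted Traveler, driving adaptive music from Baron enum events and named events.
; It spawns idle music at the spawner root, later fading it out for combat music.
; Enum events, compared with the previous one and a shields-down flag, fire trigger parameters on the combat audio component for stage transitions, death stinger and endings.
; Named events trigger endings and start a timer that fades out both tracks.
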(event MusicEncounter|EventExecuteEnumEvent
  (params (Music%20Encounter%20Component :type Music%20Encounter%20Component%20Base%20Object%20Reference) (Event%20Enum :type Byte))
  (bind _returnvalue (CallFunction|CompareOldBaronEvent "NewEnumerator3"))
  (bind _returnvalue_2 (Utilities|Enum|BytetoEnumVividRazor_Baron_Event_Enum Event%20Enum))
  ;[1 1 1] Set Music Encounter State
  (switch Utilities|FlowControl|Switch|SwitchonVividRazor_Baron_Event_Enum _returnvalue_2
    (:NewEnumerator3
      (if (not _returnvalue)
        (CallFunction|TransitiontoIdleMusic)))
    (:NewEnumerator0
      (if _returnvalue
        (if (Variables|Default|GetShieldsDown)
          (CallFunction|PlayCombatStage3)
          (else
            (CallFunction|PlayCombatMusic)))))
    (:NewEnumerator2
      (Variables|Default|SetShieldsDown true)
      (if (CallFunction|CompareOldBaronEvent "NewEnumerator0")
        _))
    (:NewEnumerator1
      (CallFunction|PlayBaronDeathStinger))
    (:NewEnumerator4
      (CallFunction|PlayTransitiontoEnding)))
  (Variables|Default|SetOldBaronEvent _returnvalue_2))

(event Custom|StartMusicEncounter
  ;[1 1 1] Start Music Encounter
  (bind _audio_component (CallFunction|SpawnAndRegisterMusic (Variables|Default|GetIdleMusic) (Variables|Default|GetSpawnerRootComponent) (Variables|MusicEncounter|MusicEventSubsystem|GetMusic%20Event%20Tag)))
  (Variables|Default|SetIdleMusicAudioComponent _audio_component))

(event MusicEncounter|EventExecuteNamedEvent
  (params (Music%20Encounter%20Component :type Music%20Encounter%20Component%20Base%20Object%20Reference) (Event%20Name :type Name :default "None"))
  ;[1 1 1] Set Music Encounter State
  (switch name Event%20Name
    (:StartEndingMusic
      (CallFunction|PlayEventEnding))
    (:Disappear
      (CallFunction|PlayEnding01)))
  (Utilities|Time|SetTimerbyEvent (EventDispatchers|CreateEvent :!event Custom|EndEvent) 1e+01))

(event Custom|PlayCombatMusic
  ;[1 1 1] Music Events
  (CallFunction|UnregisterAndFadeOutMusic (Variables|Default|GetIdleMusicAudioComponent) 2.0)
  (bind _audio_component (CallFunction|SpawnAndRegisterMusic (Variables|Default|GetCombatMusic) (Variables|Default|GetSpawnerRootComponent) (Variables|Default|GetMusicEventTag)))
  (Variables|Default|SetCombatMusicAudioComponent _audio_component)
  (CallFunction|ExecuteTriggeronAudioComponent "NPC.Event.PlayTransitionToStage2"))

(event Custom|PlayCombatStage3
  ;[1 1 1] Music Events
  (CallFunction|ExecuteTriggeronAudioComponent "NPC.Event.PlayTransitionToStage3"))

(event Custom|PlayTransitionToEnding
  ;[1 1 1] Music Events
  ;;[1 1 1] Endings
  (CallFunction|ExecuteTriggeronAudioComponent "NPC.Event.PlayTransitionToEnding"))

(event Custom|PlayEnding01
  ;[1 1 1] Music Events
  ;;[1 1 1] Endings
  (CallFunction|ExecuteTriggeronAudioComponent "NPC.Event.PlayEnding01"))

(event Custom|TransitionToIdleMusic
  ;[1 1 1] Music Events
  (CallFunction|ExecuteTriggeronAudioComponent "NPC.Event.PlayTransitionToStage1"))

(event Custom|PlayEventEnding
  ;[1 1 1] Music Events
  ;;[1 1 1] Endings
  (CallFunction|ExecuteTriggeronAudioComponent "NPC.Event.PlayEnding03"))

(event Custom|PlayBaronDeathStinger
  ;[1 1 1] Music Events
  (CallFunction|ExecuteTriggeronAudioComponent "NPC.Event.PlayEnding02"))

(event Custom|EndEvent
  ;[1 1 1] End Event
  (bind _idlemusicaudiocomponent (Variables|Default|GetIdleMusicAudioComponent)
    (:then
      (CallFunction|UnregisterAndFadeOutMusic _idlemusicaudiocomponent 5.0))
    (:else
      _))
  (bind _combatmusicaudiocomponent (Variables|Default|GetCombatMusicAudioComponent)
    (:then
      (CallFunction|UnregisterAndFadeOutMusic _combatmusicaudiocomponent 5.0))
    (:else)))

(fn CompareOldBaronEvent
  (params (B :type VividRazor_Baron_Event_Enum%20Enum :default "NewEnumerator3"))
  (return (Utilities|Enum|Equal(Enum) (Variables|Default|GetOldBaronEvent) B)))

(fn ExecuteTriggerOnAudioComponent
  (params (InName :type Name))
  (bind _combatmusicaudiocomponent (Variables|Default|GetCombatMusicAudioComponent)
    (:then
      (Audio|Parameter|ExecuteTriggerParameter _combatmusicaudiocomponent InName))
    (:else)))

(fn GetMusicManager
  (return (|CallParentFunction)))

(fn InitializeSpawnerComponent
  (params (Spawner :type Actor%20Object%20Reference))
  (Utilities|IsValid Spawner
    (:Is%20Valid
      (Utilities|FlowControl|Sequence
        (:then_0
          (bind _spawner (Variables|GetSpawner)
            (:then
              (Variables|Default|SetSpawnerRootComponent (Variables|/Script/Engine.Actor|GetRootComponent _spawner)))
            (:else)))))))

(fn SpawnandRegisterMusic
  (params (Music :type Sound%20Base%20Object%20Reference) (AttachToComponent :type Scene%20Component%20Object%20Reference) (EventTag :type Gameplay%20Tag%20Structure) (On%20Virtualization%20Changed :type Delegate%20%28by%20ref%29))
  (return 0))

(fn UnregisterandFadeOutMusic
  (params (Audio%20Component :type Audio%20Component%20Object%20Reference) (FadeOutDuration :type Float%20%28single-precision%29) (On%20Audio%20Finished :type Delegate%20%28by%20ref%29))
  (return))
